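# Physics asset for the one-handed Street Cuddles Male pickaxe skeletal mesh, a cosmetic melee weapon.
# It holds a single capsule body on the Melee_TwoHanded_Handle bone, offset from the bone origin along the capsule's length and slightly to one side.
# It has no constraints, so the weapon's physics shape is this one capsule with no jointed or simulated chain parts.
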PhysicsAsset /BRCosmetics/Weapons/FORT_Melee/Street_Cuddles_Male_1h/Meshes/SK_Pickaxe_Street_Cuddles_Male_1h_PhysicsAsset
bodies {Melee_TwoHanded_Handle [{shapeType=Capsule, length=25.9051, radius=9.438, center=<0.0005 4.2529 15.7913>, rotation=<rot 2.4322 -90 -0.0024>}]}
constraints []
